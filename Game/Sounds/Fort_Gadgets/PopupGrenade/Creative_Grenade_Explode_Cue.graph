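# Sound cue for the Creative popup grenade explosion, playing a single explosion wave at reduced volume.
# An Enveloper wraps the wave player, fading volume out over the first fraction of a second and slightly randomizing pitch and volume on each play.
# It uses the Creative prefab attenuation settings and routes to the large explosives sound class.
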SoundCue /Game/Sounds/Fort_Gadgets/PopupGrenade/Creative_Grenade_Explode_Cue
root=Enveloper_0
nodes [
 {
  id=Enveloper_0
  children [WavePlayer_0]
  properties {
   pitchMin=0.9
   volumeMax=1.05
   volumeMin=0.95
   pitchCurve {editorCurveData {keys [{time=0}]}}
   volumeCurve {editorCurveData {keys [<at 0 1>,<at 0.7 0>]}}
  }
 }
 {id=WavePlayer_0, properties {soundWaveAssetPtr=&/Creative_Grenade_Explode_01}}
]
properties {
 attenuationSettings=/Game/Sounds/Creative/Attenuations/Fort_Creative_PreFabs_Att
 maxDistance=6000
 soundClassObject=/Game/Sounds/FortSoundClassesAndModes/Fort_Explosives_Large_PSM
 volumeMultiplier=0.4
}
