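# Sound cue for the legendary popup grenade's explosion, built from Port-a-Fort legendary fast-construct recordings.
# A mixer layers a close branch and a slightly quieter distant branch, each with its own attenuation.
# In each branch a quality-level node picks between two randomized waves at the first two levels and a single optimized wave at the other two.
# A concurrency override allows two instances, stopping the farthest first.
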SoundCue /Game/Sounds/Fort_Gadgets/PopupGrenade/LegendaryPopupGrenade_Explosion_Cue
root=Mixer_0
nodes [
 {
  id=Attenuation_1
  children [QualityLevel_0]
  properties {attenuationSettings=/Game/Sounds/Attenuation/Fort_Weapon_Projectile_Close_Missile_Athena_Att}
 }
 {
  id=Attenuation_8
  children [QualityLevel_1]
  properties {attenuationSettings=&Fort_Weapon_Projectile_Medium_Mono_Athena_Att}
 }
 {id=Mixer_0, children [Attenuation_1, Attenuation_8], properties {inputVolume [0.9,0.7]}}
 {id=QualityLevel_0, children [Random_0,Random_0,WavePlayer_0,WavePlayer_0]}
 {id=QualityLevel_1, children [Random_1,Random_1,WavePlayer_1,WavePlayer_1]}
 {id=Random_0, children [WavePlayer_2,WavePlayer_3], properties {weights [1,1]}}
 {id=Random_1, children [WavePlayer_5,WavePlayer_6], properties {weights [1,1]}}
 {id=WavePlayer_0, properties {soundWaveAssetPtr=&/Port_a_Fort_Legend_Construct_Fast_02_wav_Opt}}
 {id=WavePlayer_1, properties {soundWaveAssetPtr=&/Port_a_Fort_Legend_Construct_Fast_Distant_05_Opt}}
 {id=WavePlayer_2, properties {soundWaveAssetPtr=&/Port_a_Fort_Legend_Construct_Fast_01}}
 {id=WavePlayer_3, properties {soundWaveAssetPtr=&/Port_a_Fort_Legend_Construct_Fast_02}}
 {id=WavePlayer_5, properties {soundWaveAssetPtr=&/Port_a_Fort_Legend_Construct_Fast_Distant_05}}
 {id=WavePlayer_6, properties {soundWaveAssetPtr=&/Port_a_Fort_Legend_Construct_Fast_Distant_06}}
]
properties {
 bOverrideConcurrency=true
 maxDistance=10579.2871
 soundClassObject=/Game/Sounds/FortSoundClassesAndModes/Fort_Explosives_Large_PSM
 volumeMultiplier=1
 concurrencyOverrides {maxCount=2, resolutionRule=StopFarthestThenPreventNew}
}
