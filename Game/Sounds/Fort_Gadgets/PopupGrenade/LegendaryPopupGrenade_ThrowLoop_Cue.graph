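# Sound cue for the looping throw sound of the legendary popup grenade, a Fortnite gadget item.
# A looping wave player feeds a modulator that applies a slight downward random pitch variation.
# It routes to the gadget items sound class and references a close projectile attenuation asset, though the attenuation override flag is on.
# A concurrency override prevents new instances once the limit is reached.
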SoundCue /Game/Sounds/Fort_Gadgets/PopupGrenade/LegendaryPopupGrenade_ThrowLoop_Cue
root=Modulator_0
nodes [
 {id=Modulator_0, children [WavePlayer_0], properties {pitchMax=1, pitchMin=0.98}}
 {id=WavePlayer_0, properties {bLooping=true, soundWaveAssetPtr=&/PopupGrenade_Legendary_Throw_01_Loop}}
]
properties {
 attenuationSettings=&Fort_Weapon_Projectile_Close_Athena_Att
 bOverrideAttenuation=true
 bOverrideConcurrency=true
 maxDistance=4000
 soundClassObject=&Fort_Gadgets_Items
 concurrencyOverrides {resolutionRule=PreventNew}
}
