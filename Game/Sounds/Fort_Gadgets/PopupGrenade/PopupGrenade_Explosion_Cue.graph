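# Sound cue for the Popup Grenade gadget's explosion, routed to the large explosives sound class with a concurrency override that prevents new instances at the limit.
# It mixes close and distant layers, each with its own attenuation.
# Quality switches give each layer two random variants at the first two levels and one "opt" wave at the others.
# The distant layer also blends in a quieter front-end build-upgrade UI sound.
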SoundCue /Game/Sounds/Fort_Gadgets/PopupGrenade/PopupGrenade_Explosion_Cue
root=Mixer_0
nodes [
 {
  id=Attenuation_1
  children [QualityLevel_0]
  properties {attenuationSettings=&Fort_Weapon_Projectile_Close_Athena_Att}
 }
 {
  id=Attenuation_8
  children [Mixer_1]
  properties {attenuationSettings=&Fort_Weapon_Projectile_Medium_Mono_Athena_Att}
 }
 {id=Mixer_0, children [Attenuation_1, Attenuation_8], properties {inputVolume [0.7,0.6]}}
 {id=Mixer_1, children [QualityLevel_1, WavePlayer_7], properties {inputVolume [1,0.25]}}
 {id=QualityLevel_0, children [Random_0,Random_0,WavePlayer_4,WavePlayer_4]}
 {id=QualityLevel_1, children [Random_1,Random_1,WavePlayer_5,WavePlayer_5]}
 {id=Random_0, children [WavePlayer_0,WavePlayer_1], properties {weights [1,1]}}
 {id=Random_1, children [WavePlayer_2,WavePlayer_3], properties {weights [1,1]}}
 {id=WavePlayer_0, properties {soundWaveAssetPtr=&/PopupGrenade_Explode_Close_01}}
 {id=WavePlayer_1, properties {soundWaveAssetPtr=&/PopupGrenade_Explode_Close_03}}
 {id=WavePlayer_2, properties {soundWaveAssetPtr=&/PopupGrenade_Explode_Distant_02}}
 {id=WavePlayer_3, properties {soundWaveAssetPtr=&/PopupGrenade_Explode_Distant_05}}
 {id=WavePlayer_4, properties {soundWaveAssetPtr=&/PopupGrenade_Explode_Close_01_opt}}
 {id=WavePlayer_5, properties {soundWaveAssetPtr=&/PopupGrenade_Explode_Distant_02_opt}}
 {id=WavePlayer_7, properties {soundWaveAssetPtr=/Game/Sounds/Fort_UI_Sounds/FrontEnd/Fort_Build_Upgrade_01}}
]
properties {
 attenuationSettings=/Game/Sounds/Attenuation/Fort_Explosion_Far_StW_Att
 bOverrideConcurrency=true
 maxDistance=11000
 soundClassObject=/Game/Sounds/FortSoundClassesAndModes/Fort_Explosives_Large_PSM
 volumeMultiplier=0.5
 concurrencyOverrides {resolutionRule=PreventNew}
}
