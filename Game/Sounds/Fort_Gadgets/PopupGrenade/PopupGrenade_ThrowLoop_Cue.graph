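# Looping sound cue for the popup grenade gadget's throw, in the gadgets and items sound class.
# A quality-level node plays the standard looping wave in its first two slots and the "_opt" variant in the last two.
# Inline attenuation override is on, superseding the referenced close-projectile attenuation asset.
# A concurrency override prevents new instances from starting once the limit is reached.
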SoundCue /Game/Sounds/Fort_Gadgets/PopupGrenade/PopupGrenade_ThrowLoop_Cue
root=QualityLevel_0
nodes [
 {id=QualityLevel_0, children [WavePlayer_1,WavePlayer_1,WavePlayer_0,WavePlayer_0]}
 {id=WavePlayer_0, properties {bLooping=true, soundWaveAssetPtr=&/PopupGrenade_ThrowLoop_opt}}
 {id=WavePlayer_1, properties {bLooping=true, soundWaveAssetPtr=&/PopupGrenade_ThrowLoop}}
]
properties {
 attenuationSettings=&Fort_Weapon_Projectile_Close_Athena_Att
 bOverrideAttenuation=true
 bOverrideConcurrency=true
 maxDistance=4000
 soundClassObject=&Fort_Gadgets_Items
 concurrencyOverrides {resolutionRule=PreventNew}
}
